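# Sound cue for stick impacts on enemies, filed with the flesh impact sounds.
# A Random node picks one of five equally weighted wave variations and feeds a root Modulator with default settings for pitch and volume variation.
# The cue plays in the Fort_Impacts sound class with boosted volume.
# Its own concurrency override blocks new instances once the limit is reached, rather than stopping playing ones.
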SoundCue /Game/Sounds/Fort_Impact_Sounds/Flesh/Enemy_Impact_Stick
root=Modulator_9
nodes [
 {id=Modulator_9, children [Random_17]}
 {id=Random_17, children [WavePlayer_77, WavePlayer_78, WavePlayer_79, WavePlayer_80, WavePlayer_81], properties {weights [1,1,1,1,1]}}
 {id=WavePlayer_77, properties {soundWaveAssetPtr=&/Enemy_Impact_Stick01}}
 {id=WavePlayer_78, properties {soundWaveAssetPtr=&/Enemy_Impact_Stick02}}
 {id=WavePlayer_79, properties {soundWaveAssetPtr=&/Enemy_Impact_Stick03}}
 {id=WavePlayer_80, properties {soundWaveAssetPtr=&/Enemy_Impact_Stick04}}
 {id=WavePlayer_81, properties {soundWaveAssetPtr=&/Enemy_Impact_Stick05}}
]
properties {
 bOverrideConcurrency=true
 maxDistance=2097152
 soundClassObject=&Fort_Impacts
 volumeMultiplier=2
 concurrencyOverrides {resolutionRule=PreventNew}
}
