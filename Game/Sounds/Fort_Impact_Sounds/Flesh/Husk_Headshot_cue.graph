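# Sound cue for headshot flesh impacts on Husk enemies, routed to the enemy-damaged sound class with bullet-impact-on-opponent attenuation.
# A quality-level root feeds two branches, each randomly picking an equally weighted headshot wave and varying its pitch.
# The second branch drops one of the first's six variations, and the last two quality slots are empty.
# A concurrency override caps it at two simultaneous instances, stopping the oldest.
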SoundCue /Game/Sounds/Fort_Impact_Sounds/Flesh/Husk_Headshot_cue
root=QualityLevel_0
nodes [
 {id=Modulator_1, children [Random_1], properties {pitchMax=1.2, pitchMin=0.8}}
 {id=Modulator_2, children [Random_2], properties {pitchMax=1.2, pitchMin=0.8}}
 {id=QualityLevel_0, children [Modulator_1,Modulator_2,null,null]}
 {
  id=Random_1
  children [WavePlayer_6, WavePlayer_7, WavePlayer_8, WavePlayer_9, WavePlayer_10, WavePlayer_11]
  properties {weights [1,1,1,1,1,1]}
 }
 {id=Random_2, children [WavePlayer_12, WavePlayer_13, WavePlayer_14, WavePlayer_16, WavePlayer_17], properties {weights [1,1,1,1,1]}}
 {id=WavePlayer_10, properties {soundWaveAssetPtr=&/headshots_v3_07b}}
 {id=WavePlayer_11, properties {soundWaveAssetPtr=&/headshots_v3_08b}}
 {id=WavePlayer_12, properties {soundWaveAssetPtr=&/headshots_v3_01b}}
 {id=WavePlayer_13, properties {soundWaveAssetPtr=&/headshots_v3_04b}}
 {id=WavePlayer_14, properties {soundWaveAssetPtr=&/headshots_v3_05b}}
 {id=WavePlayer_16, properties {soundWaveAssetPtr=&headshots_v3_07b}}
 {id=WavePlayer_17, properties {soundWaveAssetPtr=&headshots_v3_08b}}
 {id=WavePlayer_6, properties {soundWaveAssetPtr=&headshots_v3_01b}}
 {id=WavePlayer_7, properties {soundWaveAssetPtr=&headshots_v3_04b}}
 {id=WavePlayer_8, properties {soundWaveAssetPtr=&headshots_v3_05b}}
 {id=WavePlayer_9, properties {soundWaveAssetPtr=&/headshots_v3_06b}}
]
properties {
 attenuationSettings=/Game/Sounds/Attenuation/Fort_Bullet_ImpactOpponent_Lng_Att
 bOverrideConcurrency=true
 maxDistance=5500
 soundClassObject=/Game/Sounds/FortSoundClassesAndModes/Fort_Enemy_Damaged_PSM
 volumeMultiplier=0.7
 attenuationOverrides {falloffDistance=6000, attenuationShapeExtents {x=4000}}
 concurrencyOverrides {maxCount=2, resolutionRule=StopOldest}
}
