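# Sound cue for player flesh impact sounds, routed to the Fort_Impacts sound class and limited by a stop-oldest, one-per-owner concurrency rule.
# A quality switch selects one of two mixers, the second with fewer random variations; its other two slots are empty.
# Each mixer layers a random impact with a quieter, briefly delayed "nonspat" layer.
# Both layers use logarithmic attenuation overrides, and the delayed one adds a non-spatialized radius.
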SoundCue /Game/Sounds/Fort_Impact_Sounds/Flesh/Player_Impact_Flesh_Spatial_Cue
root=QualityLevel_0
nodes [
 {
  id=Attenuation_2
  children [Random_2]
  properties {
   bOverrideAttenuation=true
   attenuationOverrides {distanceAlgorithm=Logarithmic, falloffDistance=1200, stereoSpread=2000, attenuationShapeExtents {x=800}}
  }
 }
 {
  id=Attenuation_3
  children [Delay_1]
  properties {
   bOverrideAttenuation=true
   attenuationOverrides {distanceAlgorithm=Logarithmic, falloffDistance=1200, nonSpatializedRadiusStart=800, stereoSpread=2000, attenuationShapeExtents {x=800}}
  }
 }
 {
  id=Attenuation_4
  children [Random_4]
  properties {
   bOverrideAttenuation=true
   attenuationOverrides {distanceAlgorithm=Logarithmic, falloffDistance=1200, stereoSpread=2000, attenuationShapeExtents {x=800}}
  }
 }
 {
  id=Attenuation_5
  children [Delay_2]
  properties {
   bOverrideAttenuation=true
   attenuationOverrides {distanceAlgorithm=Logarithmic, falloffDistance=1200, nonSpatializedRadiusStart=800, stereoSpread=2000, attenuationShapeExtents {x=800}}
  }
 }
 {id=Delay_1, children [Random_3], properties {delayMax=0.1, delayMin=0.1}}
 {id=Delay_2, children [Random_5], properties {delayMax=0.1, delayMin=0.1}}
 {id=Mixer_1, children [Attenuation_2, Attenuation_3], properties {inputVolume [1,0.7]}}
 {id=Mixer_2, children [Attenuation_4, Attenuation_5], properties {inputVolume [1,0.7]}}
 {id=QualityLevel_0, children [Mixer_1,Mixer_2,null,null]}
 {
  id=Random_2
  children [WavePlayer_10, WavePlayer_11, WavePlayer_12, WavePlayer_13, WavePlayer_14, WavePlayer_15]
  properties {weights [1,1,1,1,1,1]}
 }
 {id=Random_3, children [WavePlayer_16, WavePlayer_17, WavePlayer_18, WavePlayer_19], properties {weights [1,1,1,1]}}
 {id=Random_4, children [WavePlayer_20, WavePlayer_21, WavePlayer_22, WavePlayer_24, WavePlayer_25], properties {weights [1,1,1,1,1]}}
 {id=Random_5, children [WavePlayer_26, WavePlayer_27, WavePlayer_28], properties {weights [1,1,1]}}
 {id=WavePlayer_10, properties {soundWaveAssetPtr=&/new_player_impact_01}}
 {id=WavePlayer_11, properties {soundWaveAssetPtr=&/new_player_impact_02}}
 {id=WavePlayer_12, properties {soundWaveAssetPtr=&/new_player_impact_03}}
 {id=WavePlayer_13, properties {soundWaveAssetPtr=&/new_player_impact_04}}
 {id=WavePlayer_14, properties {soundWaveAssetPtr=&/new_player_impact_05}}
 {id=WavePlayer_15, properties {soundWaveAssetPtr=&/new_player_impact_06}}
 {id=WavePlayer_16, properties {soundWaveAssetPtr=&/new_player_impact_nonspat_A_01n}}
 {id=WavePlayer_17, properties {soundWaveAssetPtr=&/new_player_impact_nonspat_A_02n}}
 {id=WavePlayer_18, properties {soundWaveAssetPtr=&/new_player_impact_nonspat_A_03n}}
 {id=WavePlayer_19, properties {soundWaveAssetPtr=&/new_player_impact_nonspat_A_04n}}
 {id=WavePlayer_20, properties {soundWaveAssetPtr=&new_player_impact_01}}
 {id=WavePlayer_21, properties {soundWaveAssetPtr=&new_player_impact_02}}
 {id=WavePlayer_22, properties {soundWaveAssetPtr=&new_player_impact_03}}
 {id=WavePlayer_24, properties {soundWaveAssetPtr=&new_player_impact_05}}
 {id=WavePlayer_25, properties {soundWaveAssetPtr=&new_player_impact_06}}
 {id=WavePlayer_26, properties {soundWaveAssetPtr=&new_player_impact_nonspat_A_01n}}
 {id=WavePlayer_27, properties {soundWaveAssetPtr=&new_player_impact_nonspat_A_02n}}
 {id=WavePlayer_28, properties {soundWaveAssetPtr=&new_player_impact_nonspat_A_03n}}
]
properties {
 maxDistance=2000
 soundClassObject=&Fort_Impacts
 volumeMultiplier=0.6
 concurrencyOverrides {resolutionRule=PreventNew}
 concurrencySet [/Game/Sounds/Concurrency_Rules/OneMaximum_Owner_StopOldest]
}
