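# Sound cue for purple flesh impact sounds, routed to the impacts sound class with monster husk attenuation.
# A quality-level root feeds two identical branches for the first two levels and leaves the other two slots empty.
# Each branch picks evenly at random from the same five impact waves, then applies slight pitch and volume modulation.
# A concurrency override blocks new instances once the limit is reached.
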SoundCue /Game/Sounds/Fort_Impact_Sounds/Flesh/Purple_Impact_Flesh_Cue
root=QualityLevel_0
nodes [
 {id=Modulator_1, children [Random_0], properties {pitchMax=1.1, pitchMin=0.9, volumeMax=1.1, volumeMin=0.9}}
 {id=Modulator_2, children [Random_3], properties {pitchMax=1.1, pitchMin=0.9, volumeMax=1.1, volumeMin=0.9}}
 {id=QualityLevel_0, children [Modulator_1,Modulator_2,null,null]}
 {id=Random_0, children [WavePlayer_0,WavePlayer_1,WavePlayer_2,WavePlayer_3,WavePlayer_4], properties {weights [1,1,1,1,1]}}
 {id=Random_3, children [WavePlayer_5, WavePlayer_6, WavePlayer_7, WavePlayer_19, WavePlayer_22], properties {weights [1,1,1,1,1]}}
 {id=WavePlayer_0, properties {soundWaveAssetPtr=&/Purple_Flesh_Impact_01}}
 {id=WavePlayer_1, properties {soundWaveAssetPtr=&/Purple_Flesh_Impact_02}}
 {id=WavePlayer_19, properties {soundWaveAssetPtr=&/Purple_Flesh_Impact_04}}
 {id=WavePlayer_2, properties {soundWaveAssetPtr=&/Purple_Flesh_Impact_03}}
 {id=WavePlayer_22, properties {soundWaveAssetPtr=&/Purple_Flesh_Impact_05}}
 {id=WavePlayer_3, properties {soundWaveAssetPtr=&Purple_Flesh_Impact_04}}
 {id=WavePlayer_4, properties {soundWaveAssetPtr=&Purple_Flesh_Impact_05}}
 {id=WavePlayer_5, properties {soundWaveAssetPtr=&Purple_Flesh_Impact_01}}
 {id=WavePlayer_6, properties {soundWaveAssetPtr=&Purple_Flesh_Impact_02}}
 {id=WavePlayer_7, properties {soundWaveAssetPtr=&Purple_Flesh_Impact_03}}
]
properties {
 attenuationSettings=/Game/Sounds/Attenuation/Fort_Monster_Husk_Att
 bOverrideConcurrency=true
 maxDistance=2800
 soundClassObject=&Fort_Impacts
 volumeMultiplier=1
 concurrencyOverrides {resolutionRule=PreventNew}
}
